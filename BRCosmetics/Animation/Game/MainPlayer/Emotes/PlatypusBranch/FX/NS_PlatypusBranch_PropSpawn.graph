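# Niagara effect for the prop spawn moment of the PlatypusBranch player emote.
# It combines three emitters: AirglowLWE, SmokeLWE and SparklesLWE, whose renderers are not listed here.
# The system uses a set-variables assignment in its spawn script, with dynamic inputs for curve-sampled color, color-by-float multiplication and a time-of-day scalar.
# User parameters expose normalized notify progress, a ribbon ID offset and a generically named vector.
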NiagaraSystem /BRCosmetics/Animation/Game/MainPlayer/Emotes/PlatypusBranch/FX/NS_PlatypusBranch_PropSpawn
dependencies {
 usedDataInterfaces [&NiagaraDataInterfaceColorCurve]
 usedDynamicInputs [
  &Multiply_LinearColorByFloat
  &ColorFromCurve
  &FNTimeOfDayScalar
 ]
 usedModules [
  &self:SystemSpawnScript.SystemScriptSource.SystemScriptGraph.NiagaraNodeAssignment_0.SetVariables_#1
  &SystemState
 ]
}
emitters {AirglowLWE {rendererClasses [], renderers []}, SmokeLWE {rendererClasses [], renderers []}, SparklesLWE {rendererClasses [], renderers []}}
summary {
 systemName=NS_PlatypusBranch_PropSpawn
 emitters [{emitterName=SparklesLWE, rendererClasses []}, {emitterName=AirglowLWE, rendererClasses []}, {emitterName=SmokeLWE, rendererClasses []}]
 userVariables [
  User.NewVector3f {type=Vector3f, description="", defaultValue {struct=Vector3f, value=<0 0 0>}}
  User.NormalizedNotifyProgress {type=NiagaraFloat, defaultValue=0, description=""}
  User.RibbonIDOffset {type=NiagaraInt32}
 ]
}
