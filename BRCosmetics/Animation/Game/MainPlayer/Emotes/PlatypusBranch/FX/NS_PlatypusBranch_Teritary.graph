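# Niagara system for a tertiary effect of the PlatypusBranch main-player emote, built from a Pebbles emitter and a SmokeLWE001 smoke emitter.
# System spawn runs two variable-assignment modules, and SystemState is also used.
# User bools flag front-end and front-end preview contexts.
# Dynamic inputs include time-of-day color, color multiplication, random float ranges and a front-end float modifier, suggesting the look adapts to lighting and menu display.
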NiagaraSystem /BRCosmetics/Animation/Game/MainPlayer/Emotes/PlatypusBranch/FX/NS_PlatypusBranch_Teritary
dependencies {
 usedDataInterfaces []
 usedDynamicInputs [
  &MultiplyLinearColor
  &MakeVector4_FromLinearColor
  &FNTimeOfDayColor
  &Vector2DFromFloat
  &FN_FrontEndFloatModifier
  &RandomRangeFloat
 ]
 usedModules [
  &self:SystemSpawnScript.SystemScriptSource.SystemScriptGraph.NiagaraNodeAssignment_0.SetVariables_#1
  &self:SystemSpawnScript.SystemScriptSource.SystemScriptGraph.NiagaraNodeAssignment_2.SetVariables_#2
  &SystemState
 ]
}
emitters {Pebbles {rendererClasses [], renderers []}, SmokeLWE001 {rendererClasses [], renderers []}}
summary {
 systemName=NS_PlatypusBranch_Teritary
 emitters [{emitterName=Pebbles, rendererClasses []}, {emitterName=SmokeLWE001, rendererClasses []}]
 userVariables [
  User.bIsFrontEnd {type=NiagaraBool}
  User.bIsFrontEndPreview {type=NiagaraBool}
 ]
}
